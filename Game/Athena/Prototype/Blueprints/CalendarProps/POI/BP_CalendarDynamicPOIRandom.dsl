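; Calendar-driven dynamic POI actor that enables one randomly chosen building foundation from a list in place of the old foundations.
; On BeginPlay, the server waits for calendar sync, optionally checks the playlist for Creative and tests a calendar event, falling back to old foundations if inactive.
; Loading can wait for the initial navmesh unlock.
; A ScalableFloat debug index can override the random pick, and LoadFoundationFromSelection performs the swap.
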(event EventBeginPlay
  (Networking|SwitchHasAuthority
    (:Authority
      (if (Variables|Default|GetDebugLoadRandomFoundation)
        (Variables|Default|SetUseCalendarEvent false)
        (else
          (MCPContext|InitialCalendarSyncCompleted
            (:OnSyncComplete
              ;[1 1 1] Need to wait for the playlist only if we want to check for creative
              (if (Variables|Default|GetIgnoreInCreative)
                (bind _gamestate (Class|FortAsyncActionCurrentPlaylistReady|PlaylistDataReadyAsync)
                  (:OnReady
                    (if (Variables|Default|GetIgnoreInCreative)
                      (if (GameplayTags|HasTag (Class|FortGameStateAthena|GetAthenaPlaylistContextTags _gamestate) "(TagName=\"Athena.Creative\")")
                        _
                        (else
                          _)))))
                (else
                  _))
              (if (Variables|Default|GetUseCalendarEvent)
                (bind _biseventactive (CalendarEvents|GetCalendarEventInformation (Variables|Default|GetCalendarEvent)))
                (if _biseventactive
                  _
                  (elif (Variables|Default|GetWaitForNavmeshUnlockBeforeLoading)
                    ;[1 1 1] Loads Default Foundations if Calendar Event is not active
                    (if (Navigation|IsInitialNavigationLockActive)
                      (Utilities|FlowControl|Delay 0.1) ; Wait for NavMesh to be Unlocked
                    )
                    (Class|BuildingFoundation|SetDynamicFoundationEnabled (Variables|Default|GetOldFoundations) true)
                    (else
                      _)))
                (else
                  _))
              (if (Variables|Default|GetWaitForNavmeshUnlockBeforeLoading)
                (if (Navigation|IsInitialNavigationLockActive)
                  (Utilities|FlowControl|Delay 0.1) ; Wait for NavMesh to be Unlocked
                )
                (else
                  _))
              (if (Utilities|Array|IsValidIndex (Variables|Default|GetFoundationList))
                (CallFunction|LoadFoundationfromSelection (Math|Integer|SelectInt (Math|Integer|Clamp(Integer) (Math|Float|Truncate (ScalableFloat|GetValueatLevel (Variables|Default|GetDebugFoundationIndex))) 0 (- (Utilities|Array|Length (Variables|Default|GetFoundationList)) 1)) (Math|Random|RandomIntegerinRange 0 (- (Utilities|Array|Length (Variables|Default|GetFoundationList)) 1)) (ScalableFloat|GetValueasBool (Variables|Default|GetUseDebugFoundationIndex))))))))))))

(event Custom|LoadFoundationFromSelection
  (params (IndexToLoad :type Integer))
  ;[1 1 1] Enable new random foundation
  (for _array_element (Variables|Default|GetOldFoundations)
    (Class|BuildingFoundation|SetDynamicFoundationEnabled _array_element false))
  (Class|BuildingFoundation|SetDynamicFoundationEnabled (Utilities|Array|Get(acopy) (Variables|Default|GetFoundationList) IndexToLoad) true))

(fn ConstructionScript)
