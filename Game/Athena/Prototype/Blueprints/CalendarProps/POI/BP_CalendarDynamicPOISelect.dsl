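; Calendar-driven POI selector that swaps a default building foundation for an event-specific dynamic foundation from a map keyed by event name.
; On authority, it watches events once the mesh network is ready and scans active ones after playlist or calendar sync, falling back to the default.
; TryLoadMap performs the swap, optionally polling until the initial navmesh lock clears, then fires a child hook.
; A debug string forces a foundation, and custom events support runtime swapping and unloading.
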(event EventBeginPlay
  (bind _debugfoundation (Variables|Default|GetDebugFoundation))
  (Networking|SwitchHasAuthority
    (:Authority
      (if (Utilities|String|EqualExactly(String) _debugfoundation)
        (Utilities|FlowControl|Sequence
          (:then_0
            (bind _nodetype (Class|FortAsyncActionMeshNetworkReady|MeshNetworkReadyAsync)
              (:OnReady
                (switch Utilities|FlowControl|Switch|SwitchonEMeshNetworkNodeType _nodetype
                  (:Root
                    _)
                  (:Inner
                    _))
                (bind _keys (Utilities|Map|Keys (Variables|Default|GetNew)))
                (if (Variables|Default|GetAlwaysLoadOnMesh)
                  (for _array_element _keys
                    (CallFunction|TryLoadMap _array_element))
                  (else
                    (bind _eventname (MCPContext|WatchCalendarEvents _keys)
                      (:OnEventActive
                        (CallFunction|TryLoadMap _eventname))))))))
          (:then_1
            (if (Variables|Default|GetIgnoreInCreative)
              (bind _gamestate (Class|FortAsyncActionCurrentPlaylistReady|PlaylistDataReadyAsync)
                (:OnReady
                  (if (GameplayTags|HasTag (Class|FortGameStateAthena|GetAthenaPlaylistContextTags _gamestate) "(TagName=\"Athena.Creative\")")
                    _
                    (else
                      _))))
              (else
                (MCPContext|InitialCalendarSyncCompleted
                  (:OnSyncComplete
                    _))))
            (bind _keys_2 (Utilities|Map|Keys (Variables|Default|GetNew)))
            (bind _array_element_2 (Utilities|Array|ForEachLoopwithBreak _keys_2)
              (:LoopBody
                (if (Variables|Default|GetLevelSetUp)
                  (CallFunction|Break)
                  (else
                    (bind _biseventactive (CalendarEvents|GetCalendarEventInformation _array_element_2))
                    (if _biseventactive
                      (CallFunction|TryLoadMap _array_element_2)))))
              (:Completed
                (if (Variables|Default|GetLevelSetUp)
                  _
                  (elif (Variables|Default|GetWaitForNavmeshUnlockBeforeLoading)
                    ;[1 1 1] Make sure NavMesh is ready to load the POI so the NavMesh is up to date
                    (if (Navigation|IsInitialNavigationLockActive) :!label _label
                      (Utilities|FlowControl|Delay 0.05) ; Wait for NavMesh to be Unlocked
                      (goto _label)
                      (else
                        _))))
                (Class|BuildingFoundation|SetDynamicFoundationEnabled (Variables|Default|GetDefaultFoundation) true)))))
        (else
          (Class|BuildingFoundation|SetDynamicFoundationEnabled (Utilities|Map|Find (Variables|Default|GetNew) _debugfoundation) true)
          (Class|BuildingFoundation|SetDynamicFoundationEnabled (Variables|Default|GetDefaultFoundation) false))))))

(event Custom|DebugCalenderEvent
  (params (A :type String))
  (bind _debugfoundation (Variables|Default|GetDebugFoundation))
  (Variables|Default|SetDebugFoundation A)
  (if (Utilities|String|EqualExactly(String) _debugfoundation)
    (Utilities|FlowControl|Sequence
      (:then_0
        (bind _nodetype (Class|FortAsyncActionMeshNetworkReady|MeshNetworkReadyAsync)
          (:OnReady
            (switch Utilities|FlowControl|Switch|SwitchonEMeshNetworkNodeType _nodetype
              (:Root
                _)
              (:Inner
                _))
            (bind _keys (Utilities|Map|Keys (Variables|Default|GetNew)))
            (if (Variables|Default|GetAlwaysLoadOnMesh)
              (for _array_element _keys
                (CallFunction|TryLoadMap _array_element))
              (else
                (bind _eventname (MCPContext|WatchCalendarEvents _keys)
                  (:OnEventActive
                    (CallFunction|TryLoadMap _eventname))))))))
      (:then_1
        (if (Variables|Default|GetIgnoreInCreative)
          (bind _gamestate (Class|FortAsyncActionCurrentPlaylistReady|PlaylistDataReadyAsync)
            (:OnReady
              (if (GameplayTags|HasTag (Class|FortGameStateAthena|GetAthenaPlaylistContextTags _gamestate) "(TagName=\"Athena.Creative\")")
                _
                (else
                  _))))
          (else
            (MCPContext|InitialCalendarSyncCompleted
              (:OnSyncComplete
                _))))
        (bind _keys_2 (Utilities|Map|Keys (Variables|Default|GetNew)))
        (bind _array_element_2 (Utilities|Array|ForEachLoopwithBreak _keys_2)
          (:LoopBody
            (if (Variables|Default|GetLevelSetUp)
              (CallFunction|Break)
              (else
                (bind _biseventactive (CalendarEvents|GetCalendarEventInformation _array_element_2))
                (if _biseventactive
                  (CallFunction|TryLoadMap _array_element_2)))))
          (:Completed
            (if (Variables|Default|GetLevelSetUp)
              _
              (elif (Variables|Default|GetWaitForNavmeshUnlockBeforeLoading)
                ;[1 1 1] Make sure NavMesh is ready to load the POI so the NavMesh is up to date
                (if (Navigation|IsInitialNavigationLockActive) :!label _label
                  (Utilities|FlowControl|Delay 0.05) ; Wait for NavMesh to be Unlocked
                  (goto _label)
                  (else
                    _))))
            (Class|BuildingFoundation|SetDynamicFoundationEnabled (Variables|Default|GetDefaultFoundation) true)))))
    (else
      (Class|BuildingFoundation|SetDynamicFoundationEnabled (Utilities|Map|Find (Variables|Default|GetNew) _debugfoundation) true)
      (Class|BuildingFoundation|SetDynamicFoundationEnabled (Variables|Default|GetDefaultFoundation) false))))

(event Custom|Break
  (bind _array_element (Utilities|Array|ForEachLoopwithBreak (stale EventBeginPlay _keys_2))
    (:LoopBody
      (if (Variables|Default|GetLevelSetUp)
        (CallFunction|Break)
        (else
          (bind _biseventactive (CalendarEvents|GetCalendarEventInformation _array_element))
          (if _biseventactive
            (CallFunction|TryLoadMap _array_element)))))
    (:Completed
      (if (Variables|Default|GetLevelSetUp)
        _
        (elif (Variables|Default|GetWaitForNavmeshUnlockBeforeLoading)
          ;[1 1 1] Make sure NavMesh is ready to load the POI so the NavMesh is up to date
          (if (Navigation|IsInitialNavigationLockActive) :!label _label
            (Utilities|FlowControl|Delay 0.05) ; Wait for NavMesh to be Unlocked
            (goto _label)
            (else
              _))))
      (Class|BuildingFoundation|SetDynamicFoundationEnabled (Variables|Default|GetDefaultFoundation) true))))

(event Custom|RuntimeFoundationSwap
  (bind _keys (Utilities|Map|Keys (Variables|Default|GetNew)))
  (for _array_element _keys
    (CallFunction|TryLoadMap _array_element)))

(event Custom|TryLoadMap
  (params (Key :type String%20%28by%20ref%29))
  (bind _new (Variables|Default|GetNew))
  (bind (_value _returnvalue) (Utilities|Map|Find _new Key))
  ;[1 1 1] TryLoadMap
  (if _returnvalue
    (Variables|Default|SetLevelSetUp true)
    (if (Variables|Default|GetWaitForNavmeshUnlockBeforeLoading)
      (if (Navigation|IsInitialNavigationLockActive) :!label _label
        (Utilities|FlowControl|Delay 0.05) ; Wait for NavMesh to be Unlocked
        (goto _label)
        (else
          _)))
    (Class|BuildingFoundation|SetDynamicFoundationEnabled (Variables|Default|GetDefaultFoundation) false)
    (Class|BuildingFoundation|SetDynamicFoundationEnabled _value true)
    (Utilities|FlowControl|DoOnce false
      (:Completed
        ;[1 1 1] Hooks for a child to extend this
        (CallFunction|DefaultFoundationDisabled)))))

(event Custom|DefaultFoundationDisabled)

(event Custom|RuntimeDynamicFoundationDisble
  (bind _keys (Utilities|Map|Keys (Variables|Default|GetNew)))
  (for _array_element _keys
    (CallFunction|UnloadDynamicLevel _array_element)))

(event Custom|UnloadDefaultLevel
  (Class|BuildingFoundation|SetDynamicFoundationEnabled (Variables|Default|GetDefaultFoundation) false))

(event Custom|UnloadDynamicLevel
  (params (Key :type String%20%28by%20ref%29))
  (bind _new (Variables|Default|GetNew))
  (bind (_value _returnvalue) (Utilities|Map|Find _new Key))
  (if _returnvalue
    (Class|BuildingFoundation|SetDynamicFoundationEnabled _value false)))

(fn OnRep_NewVar_0)

(fn ConstructionScript)
